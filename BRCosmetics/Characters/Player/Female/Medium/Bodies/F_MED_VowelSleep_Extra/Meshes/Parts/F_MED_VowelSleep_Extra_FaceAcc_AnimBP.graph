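# Anim Blueprint for the F_MED_VowelSleep_Extra face accessory part, simulating hair secondary motion that adapts to worn backbling.
# Copies the parent mesh pose through a Control Rig fed backbling and cape flags, then caches it.
# A bool set on customization load picks backbling RigidBody physics plus rear-hair AnimDynamics or Expand RigidBody physics, each with bone offsets, before shared front/side tweaks.
# Each update computes a gravity override from a parameter asset.
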AnimBlueprint /BRCosmetics/Characters/Player/Female/Medium/Bodies/F_MED_VowelSleep_Extra/Meshes/Parts/F_MED_VowelSleep_Extra_FaceAcc_AnimBP
target_skeleton=&/F_MED_VowelSleep_Extra_FaceAcc_Skeleton
graphs [
 AnimGraph {nodes [
   n1 = RigidBody!(ComponentPose=Local To Component(LocalPose=Use cached pose 'Post Deformation'()), Alpha, OverrideWorldGravity:Vector="0,0,0", ExternalForce) -> (Pose) {
    bOverrideWorldGravity=true
    overridePhysicsAsset=/BRCosmetics/Characters/Player/Female/Medium/Bodies/F_MED_VowelSleep/Meshes/Parts/F_MED_VowelSleep_FaceAcc_Backbling_Physics
    simulationSpace=BaseBoneSpace
    baseBoneRef {boneName=spine_04}
    componentAppliedLinearAccClamp=<0 0 0>
    simSpaceSettings {
     dampingAlpha=0.2
     maxAngularAcceleration=14
     maxAngularVelocity=2
     maxLinearAcceleration=2400
     maxLinearVelocity=770
     velocityScaleZ=0.4
     worldAlpha=0.8
     externalLinearDragV=<0.1 0.1 0.5>
    }
   }
   n2 = Anim Dynamics - Bone: dyn_hair_rear_l_03(ComponentPose=@n1, Alpha=1, ExternalForce="0,0,0") {
    angularDampingOverride=1
    angularSpringConstant=20000
    bAngularSpring=true
    bOverrideAngularDamping=true
    bUseGravityOverride=true
    gravityScale=0.35
    simulationSpace=RootRelative
    boundBone {boneName=dyn_hair_rear_l_03}
    componentAppliedLinearAccClamp=<0 0 0>
    gravityOverride {z=-980}
    physicsBodyDefinitions [{
      boundBone {boneName=dyn_hair_rear_l_03}
      boxExtents=<20 30 30>
      constraintSetup {angularLimitsMax=<2 10 2>, angularLimitsMin=<-2 -10 -2>, angularTarget=<1 0 0>}
      localJointOffset=<4 0 0>
     }]
    simSpaceSettings {maxAngularAcceleration=30, maxAngularVelocity=770, simSpaceAngularAlpha=1}
   }
   n3 = Anim Dynamics - Bone: dyn_hair_rear_l_04(ComponentPose=@n2, Alpha=1, ExternalForce="0,0,0") {
    angularDampingOverride=1
    angularSpringConstant=20000
    bAngularSpring=true
    bOverrideAngularDamping=true
    bUseGravityOverride=true
    gravityScale=0.35
    simulationSpace=RootRelative
    boundBone {boneName=dyn_hair_rear_l_04}
    componentAppliedLinearAccClamp=<0 0 0>
    gravityOverride {z=-980}
    physicsBodyDefinitions [{
      boundBone {boneName=dyn_hair_rear_l_04}
      boxExtents=<20 30 30>
      constraintSetup {angularLimitsMax=<2 10 2>, angularLimitsMin=<-2 -10 -2>, angularTarget=<1 0 0>}
      localJointOffset=<4 0 0>
     }]
    simSpaceSettings {maxAngularAcceleration=30, maxAngularVelocity=770, simSpaceAngularAlpha=1}
   }
   n4 = "Transform (Modify) Bone - Bone: dyn_hair_rear_l_02"(ComponentPose=@n3, Alpha=1, Translation="0,0,0", Rotation="0,0,0", Scale="0.85,0.7,1") {
    rotationMode=BMM_Additive
    rotationSpace=BCS_BoneSpace
    scaleMode=BMM_Additive
    scaleSpace=BCS_BoneSpace
    translationMode=BMM_Additive
    translationSpace=BCS_BoneSpace
    boneToModify {boneName=dyn_hair_rear_l_02}
    scale=<0.85 0.7>
   }
   n5 = "Transform (Modify) Bone - Bone: dyn_hair_rear_l_04"(ComponentPose=@n4, Alpha=1, Translation="0.755,2.957,0.087", Rotation="1.219,-31.420,-3.159", Scale="1,1,1") {
    rotationMode=BMM_Additive
    rotationSpace=BCS_BoneSpace
    scaleMode=BMM_Additive
    scaleSpace=BCS_BoneSpace
    translationMode=BMM_Additive
    translationSpace=BCS_BoneSpace
    boneToModify {boneName=dyn_hair_rear_l_04}
    rotation=<rot 1.219 -31.42 -3.159>
    translation=<0.755 2.957 0.087>
   }
   n6 = Anim Dynamics - Bone: dyn_hair_rear_03(ComponentPose=@n5, Alpha=1, ExternalForce="0,0,0") {
    angularDampingOverride=1
    angularSpringConstant=20000
    bAngularSpring=true
    bOverrideAngularDamping=true
    bUseGravityOverride=true
    gravityScale=0.35
    simulationSpace=RootRelative
    boundBone {boneName=dyn_hair_rear_03}
    componentAppliedLinearAccClamp=<0 0 0>
    gravityOverride {z=-980}
    physicsBodyDefinitions [{
      boundBone {boneName=dyn_hair_rear_03}
      boxExtents=<20 30 30>
      constraintSetup {angularLimitsMax=<2 10 2>, angularLimitsMin=<-2 -10 -2>, angularTarget=<1 0 0>}
      localJointOffset=<4 0 0>
     }]
    simSpaceSettings {maxAngularAcceleration=30, maxAngularVelocity=770, simSpaceAngularAlpha=1}
   }
   n7 = Anim Dynamics - Bone: dyn_hair_rear_04(ComponentPose=@n6, Alpha=1, ExternalForce="0,0,0") {
    angularDampingOverride=1
    angularSpringConstant=20000
    bAngularSpring=true
    bOverrideAngularDamping=true
    bUseGravityOverride=true
    gravityScale=0.35
    simulationSpace=RootRelative
    boundBone {boneName=dyn_hair_rear_04}
    componentAppliedLinearAccClamp=<0 0 0>
    gravityOverride {z=-980}
    physicsBodyDefinitions [{
      boundBone {boneName=dyn_hair_rear_04}
      boxExtents=<20 30 30>
      constraintSetup {angularLimitsMax=<2 10 2>, angularLimitsMin=<-2 -10 -2>, angularTarget=<1 0 0>}
      localJointOffset=<4 0 0>
     }]
    simSpaceSettings {maxAngularAcceleration=30, maxAngularVelocity=770, simSpaceAngularAlpha=1}
   }
   n8 = "Transform (Modify) Bone - Bone: dyn_hair_rear_03"(ComponentPose=@n7, Alpha=1, Translation="0,0,0", Rotation="0,0,0", Scale="0.85,0.6,1") {
    rotationMode=BMM_Additive
    rotationSpace=BCS_BoneSpace
    scaleMode=BMM_Additive
    scaleSpace=BCS_BoneSpace
    translationMode=BMM_Additive
    translationSpace=BCS_BoneSpace
    boneToModify {boneName=dyn_hair_rear_03}
    scale=<0.85 0.6>
   }
   n9 = "Transform (Modify) Bone - Bone: dyn_hair_rear_04"(ComponentPose=@n8, Alpha=1, Translation="-0.205,-2.779,-0.020", Rotation="0.450,0.674,-0.232", Scale="1,1,1") {
    rotationMode=BMM_Additive
    rotationSpace=BCS_BoneSpace
    scaleMode=BMM_Additive
    scaleSpace=BCS_BoneSpace
    translationMode=BMM_Additive
    translationSpace=BCS_BoneSpace
    boneToModify {boneName=dyn_hair_rear_04}
    rotation=<rot 0.45 0.674 -0.232>
    translation=<-0.205 -2.779 -0.02>
   }
   n10 = Anim Dynamics - Bone: dyn_hair_rear_r_03(ComponentPose=@n9, Alpha=1, ExternalForce="0,0,0") {
    angularDampingOverride=1
    angularSpringConstant=20000
    bAngularSpring=true
    bOverrideAngularDamping=true
    bUseGravityOverride=true
    gravityScale=0.35
    simulationSpace=RootRelative
    boundBone {boneName=dyn_hair_rear_r_03}
    componentAppliedLinearAccClamp=<0 0 0>
    gravityOverride {z=-980}
    physicsBodyDefinitions [{
      boundBone {boneName=dyn_hair_rear_r_03}
      boxExtents=<20 30 30>
      constraintSetup {angularLimitsMax=<2 10 2>, angularLimitsMin=<-2 -10 -2>, angularTarget=<1 0 0>}
      localJointOffset=<-4 0 0>
     }]
    simSpaceSettings {maxAngularAcceleration=30, maxAngularVelocity=770, simSpaceAngularAlpha=1}
   }
   n11 = Anim Dynamics - Bone: dyn_hair_rear_r_04(ComponentPose=@n10, Alpha=1, ExternalForce="0,0,0") {
    angularDampingOverride=1
    angularSpringConstant=20000
    bAngularSpring=true
    bOverrideAngularDamping=true
    bUseGravityOverride=true
    gravityScale=0.35
    simulationSpace=RootRelative
    boundBone {boneName=dyn_hair_rear_r_04}
    componentAppliedLinearAccClamp=<0 0 0>
    gravityOverride {z=-980}
    physicsBodyDefinitions [{
      boundBone {boneName=dyn_hair_rear_r_04}
      boxExtents=<20 30 30>
      constraintSetup {angularLimitsMax=<2 10 2>, angularLimitsMin=<-2 -10 -2>, angularTarget=<1 0 0>}
      localJointOffset=<-4 0 0>
     }]
    simSpaceSettings {maxAngularAcceleration=30, maxAngularVelocity=770, simSpaceAngularAlpha=1}
   }
   n12 = "Transform (Modify) Bone - Bone: dyn_hair_rear_r_02"(ComponentPose=@n11, Alpha=1, Translation="0,0,0", Rotation="0,0,0", Scale="0.85,0.7,1") {
    rotationMode=BMM_Additive
    rotationSpace=BCS_BoneSpace
    scaleMode=BMM_Additive
    scaleSpace=BCS_BoneSpace
    translationMode=BMM_Additive
    translationSpace=BCS_BoneSpace
    boneToModify {boneName=dyn_hair_rear_r_02}
    scale=<0.85 0.7>
   }
   n13 = "Transform (Modify) Bone - Bone: dyn_hair_rear_r_04"(ComponentPose=@n12, Alpha=1, Translation="0.001,0.146,0.000", Rotation="0,0,0", Scale="1,1,1") {
    rotationMode=BMM_Additive
    rotationSpace=BCS_BoneSpace
    scaleMode=BMM_Additive
    scaleSpace=BCS_BoneSpace
    translationMode=BMM_Additive
    translationSpace=BCS_BoneSpace
    boneToModify {boneName=dyn_hair_rear_r_04}
    translation=<0.001 0.146>
   }
   n14 = Component To Local(ComponentPose=@n13)
   n15 = Get isWearingBackbling()
   n16 = RigidBody!(
    ComponentPose=Local To Component(LocalPose=Use cached pose 'Post Deformation'())
    Alpha
    OverrideWorldGravity="0,0,0"
    ExternalForce
    SimSpaceSettings:Sim Space Settings Structure=(WorldAlpha=0,VelocityScaleZ=1,DampingAlpha=1,MaxLinearVelocity=10000,MaxAngularVelocity=10000,MaxLinearAcceleration=10000,MaxAngularAcceleration=10000,ExternalLinearDragV=(X=0,Y=0,Z=0),ExternalLinearVelocity=(X=0,Y=0,Z=0),ExternalAngularVelocity=(X=0,Y=0,Z=0))
   ) -> (Pose) {
    bOverrideWorldGravity=true
    overridePhysicsAsset=&/F_MED_VowelSleep_FaceAcc_Expand_Physics
    simulationSpace=BaseBoneSpace
    baseBoneRef {boneName=spine_04}
    componentAppliedLinearAccClamp=<0 0 0>
    simSpaceSettings {
     dampingAlpha=0.2
     maxAngularAcceleration=14
     maxAngularVelocity=2
     maxLinearAcceleration=2400
     maxLinearVelocity=770
     velocityScaleZ=0.5
     worldAlpha=0.9
     externalLinearDragV=<0.1 0.1 0.45>
    }
   }
   n17 = "Transform (Modify) Bone - Bone: dyn_hair_rear_l_02"(ComponentPose=@n16, Alpha=1, Translation="0,-0.7,0", Rotation="0,-2,0", Scale="1,1,1") {
    rotationMode=BMM_Additive
    rotationSpace=BCS_BoneSpace
    scaleSpace=BCS_BoneSpace
    translationMode=BMM_Additive
    translationSpace=BCS_BoneSpace
    boneToModify {boneName=dyn_hair_rear_l_02}
    rotation {yaw=-2}
    translation {y=-0.7}
   }
   n18 = "Transform (Modify) Bone - Bone: dyn_hair_rear_03"(ComponentPose=@n17, Alpha=1, Translation="0,-0.7,0", Rotation="0,-2,0", Scale="1,1,1") {
    rotationMode=BMM_Additive
    rotationSpace=BCS_BoneSpace
    scaleSpace=BCS_BoneSpace
    translationMode=BMM_Additive
    translationSpace=BCS_BoneSpace
    boneToModify {boneName=dyn_hair_rear_03}
    rotation {yaw=-2}
    translation {y=-0.7}
   }
   n19 = "Transform (Modify) Bone - Bone: dyn_hair_rear_r_02"(ComponentPose=@n18, Alpha=1, Translation="0,0.7,0", Rotation="0,-2,0", Scale="1,1,1") {
    rotationMode=BMM_Additive
    rotationSpace=BCS_BoneSpace
    scaleSpace=BCS_BoneSpace
    translationMode=BMM_Additive
    translationSpace=BCS_BoneSpace
    boneToModify {boneName=dyn_hair_rear_r_02}
    rotation {yaw=-2}
    translation {y=0.7}
   }
   n20 = "Transform (Modify) Bone - Bone: dyn_hair_front_r_01"(ComponentPose=@n19, Alpha=1, Translation="0,0.2,0", Rotation="0,0,0", Scale="1,1,1") {
    rotationMode=BMM_Additive
    rotationSpace=BCS_BoneSpace
    scaleSpace=BCS_BoneSpace
    translationMode=BMM_Additive
    translationSpace=BCS_BoneSpace
    boneToModify {boneName=dyn_hair_front_r_01}
    translation {y=0.2}
   }
   n21 = Component To Local(ComponentPose=@n20)
   n22 = "Transform (Modify) Bone - Bone: dyn_hair_front_r_03"(
    ComponentPose=Local To Component(LocalPose=Blend Poses by bool(BlendPose_0=@n14, BlendPose_1=@n21, bActiveValue=@n15))
    Alpha=1
    Translation="0,0,0"
    Rotation="10,0,0"
    Scale="1,1,1"
   ) {rotationMode=BMM_Additive, rotationSpace=BCS_BoneSpace, boneToModify {boneName=dyn_hair_front_r_03}, rotation {pitch=10}}
   n23 = "Transform (Modify) Bone - Bone: dyn_hair_front_l_03"(ComponentPose=@n22, Alpha=1, Translation="0,0,0", Rotation="10,0,0", Scale="1,1,1") {rotationMode=BMM_Additive, rotationSpace=BCS_BoneSpace, boneToModify {boneName=dyn_hair_front_l_03}, rotation {pitch=10}}
   n24 = "Transform (Modify) Bone - Bone: dyn_hair_side_l_03"(ComponentPose=@n23, Alpha=1, Translation="-2,0,0", Rotation="0,0,0", Scale="1,1,1") {
    rotationMode=BMM_Additive
    rotationSpace=BCS_BoneSpace
    translationMode=BMM_Additive
    translationSpace=BCS_BoneSpace
    boneToModify {boneName=dyn_hair_side_l_03}
    translation {x=-2}
   }
   n25 = "Transform (Modify) Bone - Bone: dyn_hair_side_r_03"(ComponentPose=@n24, Alpha=1, Translation="2,0,0", Rotation="0,0,0", Scale="1,1,1") {
    rotationMode=BMM_Additive
    rotationSpace=BCS_BoneSpace
    translationMode=BMM_Additive
    translationSpace=BCS_BoneSpace
    boneToModify {boneName=dyn_hair_side_r_03}
    translation {x=2}
   }
   Output Pose(Result=Component To Local(ComponentPose=@n25)) {layerGroup=None}
   n26 = Control Rig!(
    Source=Copy Pose From Mesh() {bUseAttachedParent=true}
    Alpha
    Is Wearing Backbling:bool=Get isWearingBackbling()
    Is Wearing Cape:bool=Get Is Backbling A Cape() -> (Is Backbling A Cape:bool)
   ) -> (Pose) {controlRigAssetReference {blueprintRigClass=&/F_MED_VowelSleep_Expand_FaceAcc_CtrlRig.F_MED_VowelSleep_Expand_FaceAcc_CtrlRig_C}}
   Save cached pose 'Post Deformation'(Pose=@n26) {cacheName=Post Deformation}
  ]}
 EventGraph {nodes [
   Event BlueprintUpdateAnimation()
   > n1 = "Gravity Override State Function (Parameter Asset) (Nativized)"(
    CharacterPartAnimInstance=Self-Reference()
    GameTimeInSeconds=GetGameTimeInSeconds()
    Parameters=/BRCosmetics/Characters/Player/Female/Medium/Bodies/F_MED_VowelSleep/Meshes/Parts/F_MED_VowelSleep_FaceAcc_Parameters
   )
   > Set Gravity Override(Gravity Override=@n1)
   Event OnCharacterCustomizationLoaded()
   n2 = CheckForBackbling()
   > Set isWearingBackbling(isWearingBackbling=OR Boolean(A=@n2, B=@n2.bIsBackblingLong))
   > Set Is Backbling A Cape(Is Backbling A Cape:bool=@n2.bIsBackblingACape) -> (Output_Get:bool)
  ]}
]
variables [
 Gravity Override {type=Vector}
 Is Backbling A Cape {type=boolean}
 isWearingBackbling {type=boolean}
 Sim Space Settings {
  type=SimSpaceSettings
  default {
   dampingAlpha=0.2
   maxAngularAcceleration=14
   maxAngularVelocity=2
   maxLinearAcceleration=2400
   maxLinearVelocity=770
   velocityScaleZ=0.5
   worldAlpha=0.9
   externalLinearDragV=<0.1 0.1 0.45>
  }
 }
]
